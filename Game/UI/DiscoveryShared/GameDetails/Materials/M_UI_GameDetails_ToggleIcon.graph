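# UI material for a toggle icon in Game Details, switching between a filled and an outline MSDF icon, with favorite icons as defaults.
# Both textures share scale-to-fit UVs, convert to SDF and get anti-aliased edges, with adjustable smoothing on the outline.
# A floored Animation Control progress value drives the blend, so the icon snaps from fill below 1 to outline at 1, the default.
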Material /Game/UI/DiscoveryShared/GameDetails/Materials/M_UI_GameDetails_ToggleIcon
graph {
 properties {bAutomaticallySetUsageInEditor=false, blendMode=BLEND_Translucent, materialDomain=MD_UI, usedShadingModels=Default Lit}
 nodes [
  n1 = MaterialFunctionCall() {materialFunction=&MF_UI_UVScaling}
  n2 = TextureSampleParameter2D(UVs=@n1.Scaled to Fit) {parameterName=Fill MSDF, samplerType=SAMPLERTYPE_LinearColor, texture=/FortUILibrary/Textures/Icons/T_UI_Icon_UIKit_Favorite_MSDF}
  n3 = Constant() {r=0.5}
  n4 = TextureSampleParameter2D(UVs=@n1.Scaled to Fit) {parameterName=Outline MSDF, samplerType=SAMPLERTYPE_LinearColor, texture=/FortUILibrary/Textures/Icons/T_UI_Icon_UIKit_FavoriteOutline_MSDF}
  LinearInterpolate(
   A=MaterialFunctionCall(
    DistanceField=MaterialFunctionCall!(MSDF RGB=@n2.RGB) {materialFunction=/FortUILibrary/MaterialFunctions/Shapes/Modifiers/MF_UI_MSDFtoSDF}.SDF
    Threshold=@n3
   ) {materialFunction=&MF_UI_AntiAliasing}.Outer
   B=MaterialFunctionCall(
    DistanceField=MaterialFunctionCall(MSDF RGB=@n4.RGB) {materialFunction=&MF_UI_MSDFtoSDF}.SDF
    Threshold=@n3
    Smoothing=ScalarParameter() {parameterName=Outline Smoothing}
   ) {materialFunction=&MF_UI_AntiAliasing}.Outer
   Alpha=Floor(None=NamedRerouteUsage())
  )
  NamedRerouteDeclaration(None=ScalarParameter() {defaultValue=1, parameterName=Animation Control}) {name=Progress, nodeColor=<rgba 0 1 0.3203 1>}
 ]
}
