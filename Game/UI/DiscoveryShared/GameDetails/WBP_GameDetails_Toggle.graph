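# WBP_GameDetails_Toggle is a UIKit generic button subclass showing a toggleable icon in game details UI, using a favorite icon by default.
# PreConstruct creates a dynamic material, feeds it fill and outline MSDF textures, aligns the icon and rounds the background corners.
# Set Toggle either snaps to the state on init or plays Anim_Toggle forward or reversed from its current time, animating material parameters and scale.
# Construct applies compact UI sizing when active.
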WidgetBlueprint /Game/UI/DiscoveryShared/GameDetails/WBP_GameDetails_Toggle
animations [Anim_Toggle {
  binding_count=1
  length_seconds=0.4888
  object=&self:Anim_Toggle
  sequence {
   bindings [ToggleableIcon {
     id=#1
     possessed_class=Image
     tracks [
      {
       type=MovieSceneWidgetMaterialTrack
       sections [{
         type=MovieSceneParameterSection
         channels [
          Outline Opacity {type=MovieSceneScriptingFloatChannel, keys [[0,1],[9,0.001],[29,0]]}
          Opacity {type=MovieSceneScriptingFloatChannel, keys [[0,0],[9,-0.0021],[29,1]]}
          Animation Control {type=MovieSceneScriptingFloatChannel, default=0, keys [[9,1],[29,0]]}
         ]
        }]
      }
      {
       type=MovieScene2DTransformTrack
       sections [{
         type=MovieScene2DTransformSection
         channels [
          Scale.X {type=MovieSceneScriptingFloatChannel, keys [[0,0.92],[9,0.492],[29,1]]}
          Scale.Y {type=MovieSceneScriptingFloatChannel, keys [[0,0.93],[9,0.4938],[29,1]]}
         ]
        }]
      }
     ]
    }]
   display_rate {numerator=60}
  }
 }]
graphs [
 EventGraph {nodes [
   n1 = Event PreConstruct()
   > Parent: PreConstruct(IsDesignTime=@n1.IsDesignTime)
   > InitVisuals()
   > SetOutlineCornerRadius()
   > SetDFTexture!(
    execute
    self
    Fill MSDF:Texture 2D Object Reference=Get Fill MSDF() -> (Fill MSDF:Texture 2D Object Reference)
    Outline MSDF:Texture 2D Object Reference=Get Outline MSDF() -> (Outline MSDF:Texture 2D Object Reference)
   ) -> (then)
   Event Construct()
   > Parent: Construct()
   > SetCompactUIStyle() {is=title K2Node_CallFunction}
  ]}
 InitVisuals {nodes [
   InitVisuals() {is=title K2Node_FunctionEntry}
   > n1 = GetDynamicMaterial(self=Get ToggleableIcon() -> (ToggleableIcon:Image Object Reference))
   > Set DynamicIconMaterial(DynamicIconMaterial:Material Instance Dynamic Object Reference=@n1) -> (Output_Get:Material Instance Dynamic Object Reference)
   n2 = SlotAsGridSlot(Widget=Get ToggleableIcon())
   > SetHorizontalAlignment(
    self:Grid Slot Object Reference=@n2
    InHorizontalAlignment=Get Icon_HorizontalAlignment() -> (Icon_HorizontalAlignment:EHorizontalAlignment Enum=HAlign_Fill)
   )
   > SetVerticalAlignment(
    self:Grid Slot Object Reference=@n2
    InVerticalAlignment=Get Icon_VerticalAlignment() -> (Icon_VerticalAlignment:EVerticalAlignment Enum=VAlign_Fill)
   )
  ]}
 Set Toggle {nodes [
   n1 = Set Toggle!() -> (then, IsInit:bool=~, IsToggled:bool=~) {is=title K2Node_FunctionEntry}
   > n2 = Set IsToggled(IsToggled:bool=@n1.IsToggled) -> (Output_Get:bool)
   > n3 = Branch(Condition=@n2)
   > n4 = Branch(Condition=Get IsInit() -> (IsInit:bool))
   > SetInitVisualState!(execute, self, IsFavorite:bool=true) -> (then)
   n5 = Get Anim_Toggle() -> (Anim_Toggle:Widget Animation Object Reference)
   PlayAnimation(
    execute=@n4.else
    InAnimation=@n5
    StartAtTime=GetAnimationCurrentTime!(self:User Widget Object Reference, InAnimation:Widget Animation Object Reference=@n5) -> (ReturnValue:float)
   )
   n6 = Branch(execute=@n3.else, Condition=Get IsInit())
   > SetInitVisualState(IsFavorite=false)
   n7 = Get Anim_Toggle()
   PlayAnimation(execute=@n6.else, InAnimation=@n7, StartAtTime=GetAnimationCurrentTime(InAnimation=@n7), PlayMode=Reverse)
  ]}
 SetCompactUIStyle {nodes [
   SetCompactUIStyle() {is=title K2Node_FunctionEntry}
   > Branch(Condition=IsUsingCompactUI())
   n1 = Get ToggleableIcon()
   > SetDesiredSizeOverride!(execute, self:Image Object Reference=@n1, DesiredSize:Vector 2D Structure=(X=52,Y=52)) -> (then)
   > SetGridSlotPadding!(execute, Widget:Widget Object Reference=@n1, Padding_Left:float=20, Padding_Top:float=16, Padding_Right:float=20, Padding_Bottom:float=16) -> (then)
  ]}
 SetDFTexture {nodes [
   n1 = SetDFTexture!() -> (then, Fill MSDF:Texture 2D Object Reference, Outline MSDF:Texture 2D Object Reference) {is=title K2Node_FunctionEntry}
   > Set Fill MSDF(Fill MSDF:Texture 2D Object Reference=@n1) -> (Output_Get:Texture 2D Object Reference)
   > Set Outline MSDF(Outline MSDF:Texture 2D Object Reference=@n1.Outline MSDF) -> (Output_Get:Texture 2D Object Reference)
   > SetTextureParameterValue(
    self=Get DynamicIconMaterial() -> (DynamicIconMaterial:Material Instance Dynamic Object Reference)
    ParameterName=Fill MSDF
    Value=Get Fill MSDF()
   )
   > SetTextureParameterValue(self=Get DynamicIconMaterial(), ParameterName=Outline MSDF, Value=Get Outline MSDF())
  ]}
 SetInitVisualState {nodes [
   SetInitVisualState!() -> (then, IsFavorite:bool=~) {is=title K2Node_FunctionEntry}
   > SetScalarParameterValue(self=Get DynamicIconMaterial(), ParameterName=Animation Control, Value=SelectFloat(A=0, B=1, bPickA=Get IsFavorite() -> (IsFavorite:bool)))
   > SetRenderScale!(
    execute
    self:Widget Object Reference=Get ToggleableIcon()
    Scale:Vector 2D Structure=Select(Option 0:Vector 2D Structure=(X=0.92,Y=0.93), Option 1:Vector 2D Structure=(X=1,Y=1), Index=Get IsFavorite()) -> (ReturnValue:Vector 2D Structure)
   ) -> (then)
  ]}
 SetOutlineCornerRadius {nodes [
   SetOutlineCornerRadius() {is=title K2Node_FunctionEntry}
   > n1 = Cast To WBP_UIKit_Block_BackgroundWithOutline!(execute, Object=Get Block_Background(None=Self-Reference()) -> (Block_Background:WBP UIKit Block Background Base Object Reference)) -> (then, CastFailed, AsWBP UIKit Block Background with Outline:WBP UIKit Block Background With Outline Object Reference) {is=title K2Node_DynamicCast}
   > SetCornerRadiuses!(
    execute
    self:WBP UIKit Block Background With Outline Object Reference=@n1
    TopLeft:float=0.325
    TopRight:float=0.325
    BottomLeft:float=0.325
    BottomRight:float=0.325
    RadiusInPixels:bool=false
   ) -> (then)
  ]}
]
variables [
 DynamicIconMaterial {type=MaterialInstanceDynamic}
 Fill MSDF {type=Texture2D, default=/FortUILibrary/Textures/Icons/T_UI_Icon_UIKit_Favorite_MSDF}
 Icon_HorizontalAlignment {type=EHorizontalAlignment, default=HAlign_Fill}
 Icon_VerticalAlignment {type=EVerticalAlignment, default=VAlign_Fill}
 IsToggled {type=boolean}
 Outline MSDF {type=Texture2D}
]
widgets {
 info {
  inheritedWidgetCount=0
  namedSlotCount=1
  parentClass=/FortUILibrary/UIKit/Buttons/WBP_UIKit_Button_Generic.WBP_UIKit_Button_Generic_C
  rootWidgetClass=null
  widgetCount=2
 }
 widgets [
  GP_ToggleIcon {widgetClassPath=&GridPanel, widget=GP_ToggleIcon, properties {columnFill [1], rowFill [1]}}
  ToggleableIcon {
   widgetClassPath=&Image
   bIsVariable=true
   parent=GP_ToggleIcon
   widget=ToggleableIcon
   properties {
    visibility=HitTestInvisible
    brush {
     bIsDynamicallyLoaded=false
     drawAs=Image
     imageType=FullColor
     mirroring=NoMirror
     resourceName=None
     resourceObject=/Game/UI/DiscoveryShared/GameDetails/Materials/M_UI_GameDetails_ToggleIcon
     tiling=NoTile
     imageSize=<32 32>
     margin=<ltrb 0 0 0 0>
     outlineSettings {
      bUseBrushTransparency=false
      roundingType=HalfHeightRadius
      width=0
      color {colorUseRule=UseColor_Specified, specifiedColor=<rgba 0 0 0 0>}
      cornerRadii=<0 0 0 0>
     }
     tintColor {colorUseRule=UseColor_Specified, specifiedColor=<rgba 1 1 1 1>}
     uVRegion {bIsValid=false, max=<0 0>, min=<0 0>}
    }
   }
   slot {class=GridSlot, properties {padding=<ltrb 16 16 16 16>}}
  }
 ]
}
